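# Animation Blueprint for the female medium HeistSleek body cosmetic, adding secondary motion and corrective curves to the pose copied from its attached parent mesh.
# The copied pose runs through the body Control Rig, then AnimDynamics simulates a front bandolier bone with angular springs and limits.
# A Pose Driver reads both thighs relative to the pelvis to drive leg-forward and crouch corrective curves, reusing StarfishUniform's skirt PoseAsset.
# The event graph is empty.
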AnimBlueprint /BRCosmetics/Characters/Player/Female/Medium/Bodies/F_MED_HeistSleek/Meshes/F_MED_HeistSleek_AnimBP
target_skeleton=&/F_MED_HeistSleek_Skeleton
graphs [
 AnimGraph {nodes [
   n1 = Control Rig(Source=Copy Pose From Mesh() {bUseAttachedParent=true}) {
    alphaBoolBlend {blendOption=Linear}
    alphaScaleBiasClamp {bInterpResult=true}
    controlRigAssetReference {blueprintRigClass=&/F_MED_HeistSleek_Body_ControlRig_new.F_MED_HeistSleek_Body_ControlRig_new_C}
   }
   n2 = Anim Dynamics - Bone: dyn_bandolier_fr_02(ComponentPose=Local To Component(LocalPose=@n1), Alpha=1, ExternalForce="0,0,0") {
    angularDampingOverride=0.99
    angularSpringConstant=10000
    bAngularSpring=true
    bGravityOverrideInSimSpace=true
    bOverrideAngularDamping=true
    numSolverIterationsPostUpdate=4
    numSolverIterationsPreUpdate=16
    alphaBoolBlend {blendOption=Linear}
    boundBone {boneName=dyn_bandolier_fr_02}
    physicsBodyDefinitions [{
      sphereCollisionRadius=0
      boundBone {boneName=dyn_bandolier_fr_02}
      boxExtents=<5 30 30>
      constraintSetup {angularTargetAxis=AxisZ, angularLimitsMax=<2 5 10>, angularLimitsMin=<-2 -20 -10>, angularTarget=<0 0 1>}
      localJointOffset=<-1 0 0>
     }]
   }
   n3 = Pose Driver - Source: thigh_l!(SourcePose:Pose Link Structure=Component To Local(ComponentPose=@n2)) -> (Pose:Pose Link Structure) {
    is=title AnimGraphNode_PoseDriver
    driveOutput=DriveCurves
    poseAsset=/BRCosmetics/Characters/Player/Female/Medium/Bodies/F_MED_StarfishUniform_01/Animations/Skirt_PoseAsset
    evalSpaceBone {boneName=pelvis}
    poseTargets [
     {
      bApplyCustomCurve=false
      bIsHidden=false
      distanceMethod=DefaultMethod
      drivenName=body_MESH_corrective_l_leg_fwd
      functionType=DefaultFunction
      targetScale=0.6
      boneTransforms [
       {targetRotation=<rot 13.2812 -50.5005 17.6642>, targetTranslation=<-1.0855 -0.2109 -8.7051>}
       {targetRotation=<rot 1.8908 -176.2781 -18.5491>, targetTranslation=<-1.0852 -0.2109 8.7051>}
      ]
      customCurve {defaultValue=3.4028234663852886e+38, postInfinityExtrap=RCCE_Constant, preInfinityExtrap=RCCE_Constant, keys []}
      targetRotation=<rot 0 0 0>
     }
     {
      bApplyCustomCurve=false
      bIsHidden=false
      distanceMethod=DefaultMethod
      drivenName=body_MESH_corrective_r_leg_fwd
      functionType=DefaultFunction
      targetScale=0.6
      boneTransforms [
       {targetRotation=<rot -4.6972 9.7725 -2.5768>, targetTranslation=<-1.0854 -0.211 -8.7049>}
       {targetRotation=<rot -0.7607 133.5814 1.7416>, targetTranslation=<-1.0853 -0.2108 8.7049>}
      ]
      customCurve {defaultValue=3.4028234663852886e+38, postInfinityExtrap=RCCE_Constant, preInfinityExtrap=RCCE_Constant, keys []}
      targetRotation=<rot 0 0 0>
     }
     {
      bApplyCustomCurve=false
      bIsHidden=false
      distanceMethod=DefaultMethod
      drivenName=body_MESH_corrective_crouch
      functionType=DefaultFunction
      targetScale=0.3467
      boneTransforms [
       {targetRotation=<rot 22.0038 -104.1937 1.0042>, targetTranslation=<-1.0855 -0.2109 -8.7053>}
       {targetRotation=<rot 19.071 148.8954 -10.517>, targetTranslation=<-1.0853 -0.2109 8.7053>}
      ]
      customCurve {defaultValue=3.4028234663852886e+38, postInfinityExtrap=RCCE_Constant, preInfinityExtrap=RCCE_Constant, keys []}
      targetRotation=<rot 0 0 0>
     }
     {
      bApplyCustomCurve=false
      bIsHidden=false
      distanceMethod=DefaultMethod
      drivenName=body_MESH_corrective_crouch_legs_apart
      functionType=DefaultFunction
      targetScale=0.62
      boneTransforms [
       {targetRotation=<rot 40.9282 -67.0934 -18.2905>, targetTranslation=<-1.0854 -0.2108 -8.7052>}
       {targetRotation=<rot 11.9869 89.3692 -25.9814>, targetTranslation=<-1.0853 -0.211 8.7052>}
      ]
      customCurve {defaultValue=3.4028234663852886e+38, postInfinityExtrap=RCCE_Constant, preInfinityExtrap=RCCE_Constant, keys []}
      targetRotation=<rot 0 0 0>
     }
     {
      bApplyCustomCurve=false
      bIsHidden=false
      distanceMethod=DefaultMethod
      drivenName=body_MESH_corrective_crouch_legs_together
      functionType=DefaultFunction
      targetScale=0.3067
      boneTransforms [
       {targetRotation=<rot -1.3353 -57.0271 57.1866>, targetTranslation=<-1.0855 -0.2109 -8.7053>}
       {targetRotation=<rot -14.1112 91.8178 35.3237>, targetTranslation=<-1.0853 -0.2109 8.7053>}
      ]
      customCurve {defaultValue=3.4028234663852886e+38, postInfinityExtrap=RCCE_Constant, preInfinityExtrap=RCCE_Constant, keys []}
      targetRotation=<rot 0 0 0>
     }
     {
      bApplyCustomCurve=false
      bIsHidden=false
      distanceMethod=DefaultMethod
      drivenName=body_MESH_corrective_crouch_l_leg_fwd
      functionType=DefaultFunction
      targetScale=0.36
      boneTransforms [
       {targetRotation=<rot 8.3274 -107.5313 -15.3861>, targetTranslation=<-1.0855 -0.2109 -8.7053>}
       {targetRotation=<rot 1.4392 133.5498 -1.197>, targetTranslation=<-1.0853 -0.2109 8.7053>}
      ]
      customCurve {defaultValue=3.4028234663852886e+38, postInfinityExtrap=RCCE_Constant, preInfinityExtrap=RCCE_Constant, keys []}
      targetRotation=<rot 0 0 0>
     }
     {
      bApplyCustomCurve=false
      bIsHidden=false
      distanceMethod=DefaultMethod
      drivenName=body_MESH_corrective_crouch_r_leg_fwd
      functionType=DefaultFunction
      targetScale=0.3067
      boneTransforms [
       {targetRotation=<rot -2.7366 -42.8404 9.3723>, targetTranslation=<-1.0853 -0.2106 -8.7046>}
       {targetRotation=<rot -0.6019 76.0806 11.9712>, targetTranslation=<-1.0852 -0.2113 8.7047>}
      ]
      customCurve {defaultValue=3.4028234663852886e+38, postInfinityExtrap=RCCE_Constant, preInfinityExtrap=RCCE_Constant, keys []}
      targetRotation=<rot 0 0 0>
     }
    ]
    rBFParams {radius=38.9391}
    sourceBones [{boneName=thigh_l}, {boneName=thigh_r}]
   }
   Output Pose(Result=@n3) {layerGroup=None}
  ]}
 EventGraph {links [], nodes []}
]
variables []
