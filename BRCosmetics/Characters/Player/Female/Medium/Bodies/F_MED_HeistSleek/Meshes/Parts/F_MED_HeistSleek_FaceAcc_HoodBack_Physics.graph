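# Physics asset for the hood-back face accessory of the HeistSleek female medium character body, giving the hood secondary motion.
# Small capsule bodies on dyn_hatLower and dyn_hatTopBack are each constrained to a capsule on the head bone.
# Both joints limit swing and twist tightly, with the lower hood piece allowed wider swing and the top back piece slightly more twist.
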PhysicsAsset /BRCosmetics/Characters/Player/Female/Medium/Bodies/F_MED_HeistSleek/Meshes/Parts/F_MED_HeistSleek_FaceAcc_HoodBack_Physics
bodies {
 dyn_hatLower [{shapeType=Capsule, length=2, radius=2, shapeName=dyn_hatLower_capsule, center=<4 0 0>, rotation=<rot 90 0 0>}]
 dyn_hatTopBack [{shapeType=Capsule, length=1, radius=1, shapeName=dyn_hatTopBack_capsule, center=<7.2178 0 0>, rotation=<rot 90 0 0>}]
 head [{shapeType=Capsule, length=2, radius=2}]
}
constraints [
 {bone1Name=dyn_hatLower, bone2Name=head, swing1LimitDegrees=25, swing2LimitDegrees=25, twistLimitDegrees=1}
 {bone1Name=dyn_hatTopBack, bone2Name=head, swing1LimitDegrees=10, swing2LimitDegrees=10, twistLimitDegrees=5}
]
