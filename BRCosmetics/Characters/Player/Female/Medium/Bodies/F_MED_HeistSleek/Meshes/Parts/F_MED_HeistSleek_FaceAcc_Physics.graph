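# Physics asset for the face accessory mesh of the F_MED_HeistSleek female medium body cosmetic, setting up physics for its hat bones.
# Capsule bodies sit on the head and the left, right and top dynamic hat bones, with each hat capsule offset along its bone.
# Each hat body is constrained directly to the head with tight swing and twist limits, the top one almost locked on one swing axis.
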PhysicsAsset /BRCosmetics/Characters/Player/Female/Medium/Bodies/F_MED_HeistSleek/Meshes/Parts/F_MED_HeistSleek_FaceAcc_Physics
bodies {
 dyn_hatLeft [{shapeType=Capsule, length=3, radius=3, shapeName=dyn_hatLeft_capsule, center=<4 0 0>, rotation=<rot 90 0 0>}]
 dyn_hatRight [{shapeType=Capsule, length=3, radius=3, shapeName=dyn_hatRight_capsule, center=<4 0 0>, rotation=<rot 90 0 0>}]
 dyn_hatTop [{shapeType=Capsule, length=3, radius=3, shapeName=dyn_hatTop_capsule, center=<4 0 0>, rotation=<rot 90 0 0>}]
 head [{shapeType=Capsule, length=2, radius=2}]
}
constraints [
 {bone1Name=dyn_hatLeft, bone2Name=head, swing1LimitDegrees=10, swing2LimitDegrees=10, twistLimitDegrees=5}
 {bone1Name=dyn_hatRight, bone2Name=head, swing1LimitDegrees=10, swing2LimitDegrees=10, twistLimitDegrees=5}
 {bone1Name=dyn_hatTop, bone2Name=head, swing1LimitDegrees=1, swing2LimitDegrees=15, twistLimitDegrees=5}
]
